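# Translucent unlit Niagara sprite and mesh material that draws comic-style speed lines for skydiving contrails.
# Opacity combines a panned line texture, cellular noise, curve-atlas and sine edge masks, and a background layer, with dynamic parameters driving offset, tiling, darkness and background strength.
# Emissive blends noise-modulated yellow and grey palettes along V, with optional particle tint.
# Optional normal-map refraction is included; mobile swaps the curve for a sine and skips depth fades.
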Material /WildEstateWeapons/Athena/Effects/Particles/Skydiving/NS_Contrail_ComicSpeedLine/FX/M_Contrail_ComicSpeedLine
graph {
 properties {
  bAutomaticallySetUsageInEditor=false
  bUsedWithNiagaraMeshParticles=true
  bUsedWithNiagaraSprites=true
  blendMode=BLEND_Translucent
  pixelDepthOffsetMode=PDOM_Legacy
  refractionMethod=RM_IndexOfRefraction
  shadingModel=MSM_Unlit
  twoSided=true
  usedShadingModels=Unlit
 }
 nodes [
  n1 = TextureCoordinate()
  n2 = DynamicParameter() {defaultValue=<rgba 0 1 1 1>, paramNames [OffsetUV,U_Tile,Darkness,BG_opacity]}
  n3 = VectorParameter() {parameterName=LineCtrl_UV_Pan, defaultValue=<rgba 1 0.05 0 0.3>}
  n4 = AppendVector(A=Multiply(A=@n2.U_Tile, B=@n3.R), B=@n3.G)
  n5 = Panner(Coordinate=Add(A=Multiply(A=@n1, B=@n4), B=@n2.OffsetUV), Speed=AppendVector(A=@n3.B, B=@n3.B))
  n6 = DynamicParameter() {parameterIndex=1, defaultValue=<rgba 0 1 1 1>, paramNames [Refraction, Refraction_Distortion, Param3, Param4]}
  n7 = Multiply(
   A=TextureSample(UVs=@n5) {samplerType=SAMPLERTYPE_Normal, texture=/Game/Athena/Apollo/Water/Textures/02_Rounded/03_WaterFoam_Veroni_base_N}.RGB
   B=@n6.Refraction_Distortion
  )
  n8 = TextureSampleParameter2D(UVs=@n5) {parameterName=LineTexture, samplerType=SAMPLERTYPE_LinearColor, texture=/Game/Gadgets/Assets/Slow_Field/Textures/T_Slowfield_SparkleNoise_01}
  n9 = LinearInterpolate(B=DynamicParameter() {defaultValue=<rgba 0 1 1 1>, paramNames [OffsetUV,U_Tile,Darkness,BG_opacity]}.Darkness, Alpha=@n8.G) {constA=-0.2, constB=5}
  n10 = TextureCoordinate()
  n11 = Multiply(A=@n10, B=Constant2Vector() {g=0.1, r=2})
  n12 = OneMinus(None=TextureSample(UVs=Panner(Coordinate=@n11) {speedY=0.5}) {samplerType=SAMPLERTYPE_Masks, texture=/Game/Global/Textures/Default/Noise/T_CellularNoise}.R)
  n13 = ComponentMask(None=@n10) {g=true}
  n14 = CurveAtlasRowParameter(CurveTime=@n13) {atlas=/Game/Athena/LinearColorCurves/TempAthenaCurveAtlas, curve=/Game/Athena/LinearColorCurves/Curve_ComicSpeedLine, defaultValue=10}
  n15 = Multiply(A=@n13) {constB=0.5}
  n16 = ComponentMask(None=@n10) {r=true}
  n17 = LinearInterpolate(B=Sine(None=Multiply(A=@n16) {constB=0.5}), Alpha=ScalarParameter() {defaultValue=1, parameterName=Use X-axis Feather}) {constA=1}
  n18 = Multiply(A=ShadingPathSwitch(Default=@n14.R, Deferred=@n14.R, Mobile=Sine(None=@n15)) {desc=Migrated from FeatureLevelSwitch}, B=@n17)
  n19 = Multiply(A=Add(A=Multiply(A=@n12, B=@n18), B=@n18)) {constB=0.5}
  n20 = ScalarParameter() {defaultValue=2, parameterName=OpacityPower_Line}
  n21 = Multiply(
   A=Multiply(A=@n17, B=@n14.G)
   B=DynamicParameter() {defaultValue=<rgba 0 1 1 1>, paramNames [OffsetUV,U_Tile,Darkness,BG_opacity]}.BG_opacity
  )
  n22 = Add(A=Multiply(A=Saturate(None=@n9), B=Power(Base=@n19, Exp=@n20)), B=@n21)
  LinearInterpolate(
   B=LinearInterpolate(B=LinearInterpolate(B=@n7, Alpha=@n22) {constA=1}, Alpha=@n6.Refraction) {constA=1}
   Alpha=ScalarParameter() {parameterName=UseRefraction}
  ) {constA=1}
  n23 = VectorParameter() {parameterName=Dark_Y, defaultValue=<rgba 2 0.7026 0 1>}
  n24 = VectorParameter() {parameterName=Bright_Y, defaultValue=<rgba 6.7574 5.5816 0 1>}
  n25 = TextureCoordinate()
  n26 = Constant2Vector() {g=0.01, r=1}
  n27 = TextureSampleParameter2D(UVs=Panner(Coordinate=Multiply(A=@n25, B=@n26)) {speedY=0.5}) {parameterName=EmissivePanner, samplerType=SAMPLERTYPE_LinearColor, texture=&HighContrastNoise}
  n28 = Multiply(A=@n27.R, B=@n27.R)
  n29 = VectorParameter() {parameterName=Dark, defaultValue=<rgba 0.2917 0.2917 0.2917 1>}
  n30 = VectorParameter() {parameterName=Bright, defaultValue=<rgba 5 5 5 1>}
  n31 = ComponentMask(None=TextureCoordinate()) {g=true}
  n32 = ScalarParameter() {defaultValue=3, parameterName=ColorLerpPower}
  n33 = ScalarParameter() {defaultValue=10, parameterName=Emissive Multiply}
  n34 = LinearInterpolate(
   B=LinearInterpolate(
    A=LinearInterpolate(A=@n23.RGB, B=@n24.RGB, Alpha=@n28)
    B=LinearInterpolate(A=@n29.RGB, B=@n30.RGB, Alpha=@n28)
    Alpha=Power(Base=@n31, Exp=@n32) {constExponent=4}
   )
   Alpha=Saturate(None=Multiply(A=@n33, B=@n21) {constB=5})
  )
  n35 = ParticleColor()
  MaterialFunctionCall(Emissive=LinearInterpolate(A=@n34, B=Multiply(A=@n34, B=@n35.RGB), Alpha=ScalarParameter() {parameterName=Use Particle Color})) {materialFunction=&MF_FX_EmissiveAdjustments}
  n36 = Multiply(
   A=DepthFade(Opacity=@n22, FadeDistance=ScalarParameter() {defaultValue=15, parameterName=DepthFadeDistance})
   B=MaterialFunctionCall(Fade Length=ScalarParameter() {defaultValue=5, parameterName=CameraDepthFade}) {materialFunction=&CameraDepthFade}.Result
  )
  Multiply(A=ShadingPathSwitch(Default=@n36, Deferred=@n36, Mobile=@n22) {desc=Migrated from FeatureLevelSwitch}, B=@n35.A)
 ]
}
